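# Unlit translucent material for a dart trail in the front-end dartboard effects, usable with beam trails.
# Particle alpha is multiplied by one minus the red (U) texture coordinate, fading the trail linearly along its length.
# That alpha feeds both a DepthFade and a DitherTemporalAA function, while particle RGB passes through the shared MF_FX_EmissiveAdjustments function.
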Material /BRCosmetics/Environments/FrontEnd/FX/Dartboard/M_SimpleDartTrail
graph {
 properties {bUsedWithBeamTrails=true, blendMode=BLEND_Translucent, pixelDepthOffsetMode=PDOM_Legacy, shadingModel=MSM_Unlit, usedShadingModels=Unlit}
 nodes [
  n1 = ParticleColor()
  n2 = TextureCoordinate()
  n3 = Multiply(A=@n1.A, B=OneMinus(None=ComponentMask(None=@n2) {r=true}))
  DepthFade(Opacity=@n3)
  MaterialFunctionCall(Emissive=@n1.RGB) {materialFunction=&MF_FX_EmissiveAdjustments}
  MaterialFunctionCall(Alpha Threshold=@n3) {materialFunction=&DitherTemporalAA}
 ]
}
